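# Physics asset for the female medium Sailor upper body (Cherry), providing capsule bodies for its dynamic scarf and the underlying torso, clavicles and upper arms.
# Three single-segment back scarf pieces connect through master bones to spine_05 and each clavicle.
# Four three-bone chains, including bkl and bkr, share the front master on spine_04.
# Master-to-body joints are angularly free, while scarf segments have per-chain swing and twist limits, tightest on fr02.
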PhysicsAsset /BRCosmetics/Characters/Player/Female/Medium/Bodies/F_MED_Pencil_B/Sailor/Meshes/Pencil_Sailor_MedB_Ubdy_Physics_Cherry
bodies {
 clavicle_l [{shapeType=Capsule, length=4, radius=5, center=<9.8201 -1.0897 0>, rotation=<rot 90 0 0>}]
 clavicle_r [{shapeType=Capsule, length=4, radius=5, center=<-9.1861 1.0034 0>, rotation=<rot 90 0 0>}]
 dyn_scarf_bk [{shapeType=Capsule, length=4.1319, radius=6.5, shapeName=dyn_scarf_bk_r_capsule, center=<7.8373 0 3.3987>, rotation=<rot 90 0 0>}]
 dyn_scarf_bk_MSTR [{shapeType=Capsule, length=0.505, radius=0.505, shapeName=dyn_scarf_bk_MSTR_capsule, rotation=<rot 0 -90 0>}]
 dyn_scarf_bk_l [{shapeType=Capsule, length=4.2965, radius=6.5, shapeName=dyn_scarf_bk_l_capsule, center=<7.8513 0 3.398>, rotation=<rot 90 0 0>}]
 dyn_scarf_bk_l_MSTR [{shapeType=Capsule, length=0.505, radius=0.505, shapeName=dyn_scarf_bk_l_MSTR_capsule, rotation=<rot 0 -90 0>}]
 dyn_scarf_bk_r [{shapeType=Capsule, length=4.1319, radius=6.5, shapeName=dyn_scarf_bk_r_capsule, center=<7.8373 0 3.323>, rotation=<rot 90 0 0>}]
 dyn_scarf_bk_r_MSTR [{shapeType=Capsule, length=0.505, radius=0.505, shapeName=dyn_scarf_bk_r_MSTR_capsule, rotation=<rot 0 -90 0>}]
 dyn_scarf_bkl_01 [{shapeType=Capsule, length=2.25, radius=2.25, shapeName=dyn_scarf_bkl_01_capsule, center=<2.545 0 2>, rotation=<rot 90 0 0>}]
 dyn_scarf_bkl_02 [{shapeType=Capsule, length=2.25, radius=2.25, shapeName=dyn_scarf_bkl_02_capsule, center=<2.545 0 2>, rotation=<rot 90 0 0>}]
 dyn_scarf_bkl_03 [{shapeType=Capsule, length=1.75, radius=2.25, shapeName=dyn_scarf_bkl_03_capsule, center=<2.545 0 2>, rotation=<rot 90 0 0>}]
 dyn_scarf_bkr_01 [{shapeType=Capsule, length=2.25, radius=2.25, shapeName=dyn_scarf_bkr_01_capsule, center=<2.545 0 2>, rotation=<rot 90 0 0>}]
 dyn_scarf_bkr_02 [{shapeType=Capsule, length=2.25, radius=2.25, shapeName=dyn_scarf_bkr_02_capsule, center=<2.545 0 2>, rotation=<rot 90 0 0>}]
 dyn_scarf_bkr_03 [{shapeType=Capsule, length=1, radius=2.25, shapeName=dyn_scarf_bkr_03_capsule, center=<2.0499 0 2>, rotation=<rot 90 0 0>}]
 dyn_scarf_fr02_01 [{shapeType=Capsule, length=2.25, radius=2.25, shapeName=dyn_scarf_fr02_01_capsule, center=<2.545 0 2>, rotation=<rot 90 0 0>}]
 dyn_scarf_fr02_02 [{shapeType=Capsule, length=2.25, radius=2.25, shapeName=dyn_scarf_fr02_02_capsule, center=<2.545 0 2>, rotation=<rot 90 0 0>}]
 dyn_scarf_fr02_03 [{shapeType=Capsule, length=1.5, radius=2.25, shapeName=dyn_scarf_fr02_03_capsule, center=<2.2988 0 2>, rotation=<rot 90 0 0>}]
 dyn_scarf_fr_01 [{shapeType=Capsule, length=2.25, radius=2.25, shapeName=dyn_scarf_fr_01_capsule, center=<2.545 0 2>, rotation=<rot 90 0 0>}]
 dyn_scarf_fr_02 [{shapeType=Capsule, length=2.25, radius=2.25, shapeName=dyn_scarf_fr_02_capsule, center=<2.545 0 2>, rotation=<rot 90 0 0>}]
 dyn_scarf_fr_03 [{shapeType=Capsule, length=1.5, radius=2.25, shapeName=dyn_scarf_fr_03_capsule, center=<2.2988 0 2>, rotation=<rot 90 0 0>}]
 dyn_scarf_fr_MSTR [{shapeType=Capsule, length=0.505, radius=0.505, shapeName=dyn_scarf_fr_MSTR_capsule, rotation=<rot 0 -90 0>}]
 spine_03 [
  {shapeType=Capsule, length=10.3031, radius=8.4883, center=<-2.4706 -1.021 0>}
  {shapeType=Capsule, length=10.5531, radius=8.2383, shapeName=spine_03_capsule, center=<-3.2726 3.3753 0>}
 ]
 spine_04 [
  {shapeType=Capsule, length=5.1831, radius=10.3838, center=<0 -0.1433 0>}
  {shapeType=Capsule, length=5.4331, radius=10.1338, shapeName=spine_04_capsule, center=<-1.6015 3.4759 0>}
 ]
 spine_05 [{shapeType=Capsule, length=10.6709, radius=7, center=<2.5199 -1.2341 -0.2457>, rotation=<rot 0 -176.6699 -178.4916>}]
 upperarm_l [{shapeType=Capsule, length=10.025, radius=7.644, shapeName=upperarm_l_capsule, center=<11.1511 -0.722 0>, rotation=<rot 90 0 0>}]
 upperarm_r [{shapeType=Capsule, length=10.275, radius=7.394, shapeName=upperarm_r_capsule, center=<-12.1845 0 0>, rotation=<rot 90 0 0>}]
}
constraints [
 {bone1Name=spine_05, bone2Name=spine_04, swing1LimitDegrees=45, swing2LimitDegrees=45, twistLimitDegrees=45}
 {bone1Name=clavicle_l, bone2Name=spine_05, swing1LimitDegrees=45, swing2LimitDegrees=45, twistLimitDegrees=45}
 {bone1Name=clavicle_r, bone2Name=spine_05, swing1LimitDegrees=45, swing2LimitDegrees=45, twistLimitDegrees=45}
 {bone1Name=dyn_scarf_bk_r, bone2Name=dyn_scarf_bk_r_MSTR, swing1LimitDegrees=45, swing2LimitDegrees=45, twistLimitDegrees=10}
 {bone1Name=dyn_scarf_bk_l, bone2Name=dyn_scarf_bk_l_MSTR, swing1LimitDegrees=45, swing2LimitDegrees=45, twistLimitDegrees=10}
 {
  bone1Name=dyn_scarf_bk_l_MSTR
  bone2Name=clavicle_l
  swing1LimitDegrees=45
  swing1Motion=Free
  swing2LimitDegrees=45
  swing2Motion=Free
  twistLimitDegrees=45
  twistMotion=Free
 }
 {
  bone1Name=dyn_scarf_bk_r_MSTR
  bone2Name=clavicle_r
  swing1LimitDegrees=45
  swing1Motion=Free
  swing2LimitDegrees=45
  swing2Motion=Free
  twistLimitDegrees=45
  twistMotion=Free
 }
 {
  bone1Name=spine_04
  bone2Name=spine_03
  swing1LimitDegrees=45
  swing1Motion=Free
  swing2LimitDegrees=45
  swing2Motion=Free
  twistLimitDegrees=45
  twistMotion=Free
 }
 {
  bone1Name=upperarm_l
  bone2Name=clavicle_l
  swing1LimitDegrees=45
  swing1Motion=Free
  swing2LimitDegrees=45
  swing2Motion=Free
  twistLimitDegrees=45
  twistMotion=Free
 }
 {
  bone1Name=upperarm_r
  bone2Name=clavicle_r
  swing1LimitDegrees=45
  swing1Motion=Free
  swing2LimitDegrees=45
  swing2Motion=Free
  twistLimitDegrees=45
  twistMotion=Free
 }
 {bone1Name=dyn_scarf_bkl_01, bone2Name=dyn_scarf_fr_MSTR, swing1LimitDegrees=20, swing2LimitDegrees=20, twistLimitDegrees=10}
 {bone1Name=dyn_scarf_bkl_02, bone2Name=dyn_scarf_bkl_01, swing1LimitDegrees=20, swing2LimitDegrees=20, twistLimitDegrees=10}
 {bone1Name=dyn_scarf_bkl_03, bone2Name=dyn_scarf_bkl_02, swing1LimitDegrees=20, swing2LimitDegrees=20, twistLimitDegrees=10}
 {bone1Name=dyn_scarf_bkr_01, bone2Name=dyn_scarf_fr_MSTR, swing1LimitDegrees=45, swing2LimitDegrees=45, twistLimitDegrees=20}
 {bone1Name=dyn_scarf_bkr_02, bone2Name=dyn_scarf_bkr_01, swing1LimitDegrees=45, swing2LimitDegrees=45, twistLimitDegrees=20}
 {bone1Name=dyn_scarf_bkr_03, bone2Name=dyn_scarf_bkr_02, swing1LimitDegrees=45, swing2LimitDegrees=45, twistLimitDegrees=20}
 {bone1Name=dyn_scarf_fr02_01, bone2Name=dyn_scarf_fr_MSTR, swing1LimitDegrees=10, swing2LimitDegrees=10, twistLimitDegrees=5}
 {bone1Name=dyn_scarf_fr02_02, bone2Name=dyn_scarf_fr02_01, swing1LimitDegrees=10, swing2LimitDegrees=10, twistLimitDegrees=5}
 {bone1Name=dyn_scarf_fr02_03, bone2Name=dyn_scarf_fr02_02, swing1LimitDegrees=10, swing2LimitDegrees=10, twistLimitDegrees=5}
 {bone1Name=dyn_scarf_fr_01, bone2Name=dyn_scarf_fr_MSTR, swing1LimitDegrees=45, swing2LimitDegrees=45, twistLimitDegrees=20}
 {bone1Name=dyn_scarf_fr_02, bone2Name=dyn_scarf_fr_01, swing1LimitDegrees=45, swing2LimitDegrees=45, twistLimitDegrees=20}
 {bone1Name=dyn_scarf_fr_03, bone2Name=dyn_scarf_fr_02, swing1LimitDegrees=45, swing2LimitDegrees=45, twistLimitDegrees=20}
 {
  bone1Name=dyn_scarf_fr_MSTR
  bone2Name=spine_04
  swing1LimitDegrees=45
  swing1Motion=Free
  swing2LimitDegrees=45
  swing2Motion=Free
  twistLimitDegrees=45
  twistMotion=Free
 }
 {
  bone1Name=dyn_scarf_bk_MSTR
  bone2Name=spine_05
  swing1LimitDegrees=45
  swing1Motion=Free
  swing2LimitDegrees=45
  swing2Motion=Free
  twistLimitDegrees=45
  twistMotion=Free
 }
 {bone1Name=dyn_scarf_bk, bone2Name=dyn_scarf_bk_MSTR, swing1LimitDegrees=45, swing2LimitDegrees=45, twistLimitDegrees=10}
]
