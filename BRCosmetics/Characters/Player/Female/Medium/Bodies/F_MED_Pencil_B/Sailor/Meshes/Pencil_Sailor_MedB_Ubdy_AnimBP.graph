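# Post-process style AnimBlueprint for the Sailor female body cosmetic, adjusting its scarf bones to motion and other worn parts.
# It copies the parent pose through deform and skirt Control Rigs, RigidBody physics and scarf offsets into a cached pose.
# Head meta tags (Guava, Pear, Fig, Cherry) pick offset, constraint or alternate-physics fixes.
# Backbling then pins and shrinks back scarf bones; EventGraph updates these flags and movement values.
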AnimBlueprint /BRCosmetics/Characters/Player/Female/Medium/Bodies/F_MED_Pencil_B/Sailor/Meshes/Pencil_Sailor_MedB_Ubdy_AnimBP
target_skeleton=&/Pencil_Sailor_MedB_Ubdy_Skeleton
graphs [
 AnimGraph {nodes [
   Input Pose()
   "Transform (Modify) Bone - Bone: dyn_scarf_bk_r"(Alpha=Get Arm Rot Up R() -> (Arm Rot Up R:double), Translation="0,3,0", Rotation="30,0,0", Scale="0.8,1,1") {
    rotationMode=BMM_Additive
    rotationSpace=BCS_BoneSpace
    scaleMode=BMM_Additive
    scaleSpace=BCS_BoneSpace
    translationMode=BMM_Additive
    translationSpace=BCS_BoneSpace
    alphaBoolBlend {blendOption=Linear}
    boneToModify {boneName=dyn_scarf_bk_r}
    rotation {pitch=30}
    scale {x=0.8}
    translation {y=3}
   }
   n1 = "Transform (Modify) Bone - Bone: dyn_scarf_bk_r"!(ComponentPose, bAlphaBoolEnabled:bool=True, Translation="0,0,2", Rotation="0,0,0", Scale="1,1,1") -> (Pose) {
    alphaInputType=Bool
    rotationMode=BMM_Additive
    rotationSpace=BCS_BoneSpace
    scaleMode=BMM_Additive
    scaleSpace=BCS_BoneSpace
    translationMode=BMM_Additive
    translationSpace=BCS_BoneSpace
    alphaBoolBlend {blendInTime=0.1, blendOption=Linear, blendOutTime=0.1}
    boneToModify {boneName=dyn_scarf_bk_r}
    translation {z=2}
   }
   n2 = "Transform (Modify) Bone - Bone: dyn_scarf_bk_l"!(ComponentPose=@n1, bAlphaBoolEnabled:bool=True, Translation="0,0,2", Rotation="0,0,0", Scale="1,1,1") -> (Pose) {
    alphaInputType=Bool
    rotationMode=BMM_Additive
    rotationSpace=BCS_BoneSpace
    scaleMode=BMM_Additive
    scaleSpace=BCS_BoneSpace
    translationMode=BMM_Additive
    translationSpace=BCS_BoneSpace
    alphaBoolBlend {blendInTime=0.1, blendOption=Linear, blendOutTime=0.1}
    boneToModify {boneName=dyn_scarf_bk_l}
    translation {z=2}
   }
   n3 = "Transform (Modify) Bone - Bone: dyn_scarf_bk"!(ComponentPose=@n2, bAlphaBoolEnabled:bool=True, Translation="0,0,2", Rotation="0,0,0", Scale="1,1,1") -> (Pose) {
    alphaInputType=Bool
    rotationMode=BMM_Additive
    rotationSpace=BCS_BoneSpace
    scaleMode=BMM_Additive
    scaleSpace=BCS_BoneSpace
    translationMode=BMM_Additive
    translationSpace=BCS_BoneSpace
    alphaBoolBlend {blendInTime=0.1, blendOption=Linear, blendOutTime=0.1}
    boneToModify {boneName=dyn_scarf_bk}
    translation {z=2}
   }
   n4 = "Transform (Modify) Bone - Bone: dyn_scarf_bk_r"(ComponentPose=@n3, bAlphaBoolEnabled=True, Translation="0,0,2", Rotation="2,0,0", Scale="1,1,1") {
    alphaInputType=Bool
    rotationMode=BMM_Additive
    rotationSpace=BCS_BoneSpace
    scaleMode=BMM_Additive
    scaleSpace=BCS_BoneSpace
    translationMode=BMM_Additive
    translationSpace=BCS_BoneSpace
    alphaBoolBlend {blendInTime=0.1, blendOption=Linear, blendOutTime=0.1}
    boneToModify {boneName=dyn_scarf_bk_r}
    rotation {pitch=2}
    translation {z=2}
   }
   n5 = "Transform (Modify) Bone - Bone: dyn_scarf_bk_l"(ComponentPose=@n4, bAlphaBoolEnabled=True, Translation="0,0,2", Rotation="2,0,0", Scale="1,1,1") {
    alphaInputType=Bool
    rotationMode=BMM_Additive
    rotationSpace=BCS_BoneSpace
    scaleMode=BMM_Additive
    scaleSpace=BCS_BoneSpace
    translationMode=BMM_Additive
    translationSpace=BCS_BoneSpace
    alphaBoolBlend {blendInTime=0.1, blendOption=Linear, blendOutTime=0.1}
    boneToModify {boneName=dyn_scarf_bk_l}
    rotation {pitch=2}
    translation {z=2}
   }
   n6 = "Transform (Modify) Bone - Bone: dyn_scarf_bk"(ComponentPose=@n5, bAlphaBoolEnabled=True, Translation="0,0,2", Rotation="2,0,0", Scale="1,1,1") {
    alphaInputType=Bool
    rotationMode=BMM_Additive
    rotationSpace=BCS_BoneSpace
    scaleMode=BMM_Additive
    scaleSpace=BCS_BoneSpace
    translationMode=BMM_Additive
    translationSpace=BCS_BoneSpace
    alphaBoolBlend {blendInTime=0.1, blendOption=Linear, blendOutTime=0.1}
    boneToModify {boneName=dyn_scarf_bk}
    rotation {pitch=2}
    translation {z=2}
   }
   n7 = "Transform (Modify) Bone - Bone: dyn_scarf_bk_r"(ComponentPose=@n6, bAlphaBoolEnabled=True, Translation="0,0,1.5", Rotation="0,0,0", Scale="1,1,1") {
    alphaInputType=Bool
    rotationMode=BMM_Additive
    rotationSpace=BCS_BoneSpace
    scaleMode=BMM_Additive
    scaleSpace=BCS_BoneSpace
    translationMode=BMM_Additive
    translationSpace=BCS_BoneSpace
    alphaBoolBlend {blendInTime=0.1, blendOption=Linear, blendOutTime=0.1}
    boneToModify {boneName=dyn_scarf_bk_r}
    translation {z=1.5}
   }
   n8 = "Transform (Modify) Bone - Bone: dyn_scarf_bk_l"(ComponentPose=@n7, bAlphaBoolEnabled=True, Translation="0,0,1.5", Rotation="0,0,0", Scale="1,1,1") {
    alphaInputType=Bool
    rotationMode=BMM_Additive
    rotationSpace=BCS_BoneSpace
    scaleMode=BMM_Additive
    scaleSpace=BCS_BoneSpace
    translationMode=BMM_Additive
    translationSpace=BCS_BoneSpace
    alphaBoolBlend {blendInTime=0.1, blendOption=Linear, blendOutTime=0.1}
    boneToModify {boneName=dyn_scarf_bk_l}
    translation {z=1.5}
   }
   n9 = "Transform (Modify) Bone - Bone: dyn_scarf_bk"(ComponentPose=@n8, bAlphaBoolEnabled=True, Translation="0,0,1.5", Rotation="0,0,0", Scale="1,1,1") {
    alphaInputType=Bool
    rotationMode=BMM_Additive
    rotationSpace=BCS_BoneSpace
    scaleMode=BMM_Additive
    scaleSpace=BCS_BoneSpace
    translationMode=BMM_Additive
    translationSpace=BCS_BoneSpace
    alphaBoolBlend {blendInTime=0.1, blendOption=Linear, blendOutTime=0.1}
    boneToModify {boneName=dyn_scarf_bk}
    translation {z=1.5}
   }
   n10 = "Transform (Modify) Bone - Bone: dyn_scarf_bk_l"(ComponentPose=@n9, bAlphaBoolEnabled=True, Translation="0,0,1", Rotation="-1,0,0", Scale="1,1,1") {
    alphaInputType=Bool
    rotationMode=BMM_Additive
    rotationSpace=BCS_BoneSpace
    scaleMode=BMM_Additive
    scaleSpace=BCS_BoneSpace
    translationMode=BMM_Additive
    translationSpace=BCS_BoneSpace
    alphaBoolBlend {blendInTime=0.1, blendOption=Linear, blendOutTime=0.1}
    boneToModify {boneName=dyn_scarf_bk_l}
    rotation {pitch=-1}
    translation {z=1}
   }
   n11 = "Transform (Modify) Bone - Bone: dyn_scarf_bk_r"(ComponentPose=@n10, bAlphaBoolEnabled=True, Translation="0,0,1", Rotation="-1,0,0", Scale="1,1,1") {
    alphaInputType=Bool
    rotationMode=BMM_Additive
    rotationSpace=BCS_BoneSpace
    scaleMode=BMM_Additive
    scaleSpace=BCS_BoneSpace
    translationMode=BMM_Additive
    translationSpace=BCS_BoneSpace
    alphaBoolBlend {blendInTime=0.1, blendOption=Linear, blendOutTime=0.1}
    boneToModify {boneName=dyn_scarf_bk_r}
    rotation {pitch=-1}
    translation {z=1}
   }
   n12 = "Transform (Modify) Bone - Bone: dyn_scarf_bk_r"!(ComponentPose=@n11, Alpha:float=Get strafe r() -> (strafe r:double), Translation="0,0,1", Rotation="0,10,0", Scale="1,1,1") -> (Pose) {
    rotationMode=BMM_Additive
    rotationSpace=BCS_BoneSpace
    scaleMode=BMM_Additive
    scaleSpace=BCS_BoneSpace
    translationMode=BMM_Additive
    translationSpace=BCS_BoneSpace
    alphaBoolBlend {blendOption=Linear}
    boneToModify {boneName=dyn_scarf_bk_r}
    rotation {yaw=10}
    translation {z=1}
   }
   "Transform (Modify) Bone - Bone: dyn_scarf_bk_l"!(ComponentPose=@n12, Alpha:float=Get strafe l() -> (strafe l:double), Translation="0,0,1", Rotation="0,10,0", Scale="1,1,1") -> (Pose) {
    rotationMode=BMM_Additive
    rotationSpace=BCS_BoneSpace
    scaleMode=BMM_Additive
    scaleSpace=BCS_BoneSpace
    translationMode=BMM_Additive
    translationSpace=BCS_BoneSpace
    alphaBoolBlend {blendOption=Linear}
    boneToModify {boneName=dyn_scarf_bk_l}
    rotation {yaw=10}
    translation {z=1}
   }
   n13 = "Transform (Modify) Bone - Bone: dyn_scarf_bk_r"(Alpha=Get strafe r(), Translation="0,0,1", Rotation="0,5,0", Scale="1,1,1") {
    rotationMode=BMM_Additive
    rotationSpace=BCS_BoneSpace
    scaleMode=BMM_Additive
    scaleSpace=BCS_BoneSpace
    translationMode=BMM_Additive
    translationSpace=BCS_BoneSpace
    alphaBoolBlend {blendOption=Linear}
    alphaScaleBiasClamp {bInterpResult=true}
    boneToModify {boneName=dyn_scarf_bk_r}
    rotation {yaw=5}
    translation {z=1}
   }
   n14 = "Transform (Modify) Bone - Bone: dyn_scarf_bk_l"(ComponentPose=@n13, Alpha=Get strafe l(), Translation="0,0,1", Rotation="0,5,0", Scale="1,1,1") {
    rotationMode=BMM_Additive
    rotationSpace=BCS_BoneSpace
    scaleMode=BMM_Additive
    scaleSpace=BCS_BoneSpace
    translationMode=BMM_Additive
    translationSpace=BCS_BoneSpace
    alphaBoolBlend {blendOption=Linear}
    alphaScaleBiasClamp {bInterpResult=true}
    boneToModify {boneName=dyn_scarf_bk_l}
    rotation {yaw=5}
    translation {z=1}
   }
   n15 = "Transform (Modify) Bone - Bone: dyn_scarf_bk_r"(ComponentPose=@n14, Alpha=Get strafe l(), Translation="0,0,1", Rotation="0,5,0", Scale="1,1,1") {
    rotationMode=BMM_Additive
    rotationSpace=BCS_BoneSpace
    scaleMode=BMM_Additive
    scaleSpace=BCS_BoneSpace
    translationMode=BMM_Additive
    translationSpace=BCS_BoneSpace
    alphaBoolBlend {blendOption=Linear}
    alphaScaleBiasClamp {bInterpResult=true}
    boneToModify {boneName=dyn_scarf_bk_r}
    rotation {yaw=5}
    translation {z=1}
   }
   "Transform (Modify) Bone - Bone: dyn_scarf_bk_l"(ComponentPose=@n15, Alpha=Get strafe r(), Translation="0,0,1", Rotation="0,5,0", Scale="1,1,1") {
    rotationMode=BMM_Additive
    rotationSpace=BCS_BoneSpace
    scaleMode=BMM_Additive
    scaleSpace=BCS_BoneSpace
    translationMode=BMM_Additive
    translationSpace=BCS_BoneSpace
    alphaBoolBlend {blendOption=Linear}
    alphaScaleBiasClamp {bInterpResult=true}
    boneToModify {boneName=dyn_scarf_bk_l}
    rotation {yaw=5}
    translation {z=1}
   }
   "Transform (Modify) Bone - Bone: dyn_scarf_bk_l"(Alpha=Get arm_rot_up_L() -> (arm_rot_up_L:double), Translation="0,3,0", Rotation="30,0,0", Scale="0.8,1,1") {
    rotationMode=BMM_Additive
    rotationSpace=BCS_BoneSpace
    scaleMode=BMM_Additive
    scaleSpace=BCS_BoneSpace
    translationMode=BMM_Additive
    translationSpace=BCS_BoneSpace
    alphaBoolBlend {blendOption=Linear}
    boneToModify {boneName=dyn_scarf_bk_l}
    rotation {pitch=30}
    scale {x=0.8}
    translation {y=3}
   }
   n16 = RigidBody!(ComponentPose=Local To Component(LocalPose=Use cached pose 'PreCach'()), Alpha, OverrideWorldGravity:Vector="0,0,-980", ExternalForce) -> (Pose) {
    bClampLinearTranslationLimitToRefPose=true
    bOverrideWorldGravity=true
    overridePhysicsAsset=&/Pencil_Sailor_MedB_Ubdy_Physics_Cherry
    simulationSpace=BaseBoneSpace
    alphaBoolBlend {blendOption=Linear}
    baseBoneRef {boneName=spine_05}
    componentAppliedLinearAccClamp=<3000 3000 3000>
    componentLinearAccScale=<0.48 0.48 0.6>
    componentLinearVelScale=<0.24 0.24 0.3>
    overrideWorldGravity {z=-980}
    simSpaceSettings {maxAngularVelocity=3, maxLinearAcceleration=700, maxLinearVelocity=500, velocityScaleZ=0, worldAlpha=0.6, externalLinearDragV=<0.2 0.2>}
   }
   n17 = "Transform (Modify) Bone - Bone: dyn_scarf_bk_l_MSTR"!(ComponentPose=@n16, bAlphaBoolEnabled:bool=True, Translation="0,0,3", Rotation="-55,0,0", Scale="1,1,1") -> (Pose) {
    alphaInputType=Bool
    rotationMode=BMM_Additive
    rotationSpace=BCS_BoneSpace
    scaleMode=BMM_Additive
    scaleSpace=BCS_BoneSpace
    translationMode=BMM_Additive
    translationSpace=BCS_BoneSpace
    alphaBoolBlend {blendInTime=0.1, blendOption=Linear, blendOutTime=0.1}
    boneToModify {boneName=dyn_scarf_bk_l_MSTR}
    rotation {pitch=-55}
    translation {z=3}
   }
   n18 = "Transform (Modify) Bone - Bone: dyn_scarf_bk_MSTR"!(ComponentPose=@n17, bAlphaBoolEnabled:bool=True, Translation="0,0,3", Rotation="-55,0,0", Scale="1,1,1") -> (Pose) {
    alphaInputType=Bool
    rotationMode=BMM_Additive
    rotationSpace=BCS_BoneSpace
    scaleMode=BMM_Additive
    scaleSpace=BCS_BoneSpace
    translationMode=BMM_Additive
    translationSpace=BCS_BoneSpace
    alphaBoolBlend {blendInTime=0.1, blendOption=Linear, blendOutTime=0.1}
    boneToModify {boneName=dyn_scarf_bk_MSTR}
    rotation {pitch=-55}
    translation {z=3}
   }
   n19 = "Transform (Modify) Bone - Bone: dyn_scarf_bk_r_MSTR"!(ComponentPose=@n18, bAlphaBoolEnabled:bool=True, Translation="0,0,3", Rotation="-55,0,0", Scale="1,1,1") -> (Pose) {
    alphaInputType=Bool
    rotationMode=BMM_Additive
    rotationSpace=BCS_BoneSpace
    scaleMode=BMM_Additive
    scaleSpace=BCS_BoneSpace
    translationMode=BMM_Additive
    translationSpace=BCS_BoneSpace
    alphaBoolBlend {blendInTime=0.1, blendOption=Linear, blendOutTime=0.1}
    boneToModify {boneName=dyn_scarf_bk_r_MSTR}
    rotation {pitch=-55}
    translation {z=3}
   }
   n20 = "Transform (Modify) Bone - Bone: dyn_scarf_bk_l"!(ComponentPose=@n19, bAlphaBoolEnabled:bool=True, Translation="0,0,1", Rotation="-1,0,0", Scale="1,1,1") -> (Pose) {
    alphaInputType=Bool
    rotationMode=BMM_Additive
    rotationSpace=BCS_BoneSpace
    scaleMode=BMM_Additive
    scaleSpace=BCS_BoneSpace
    translationMode=BMM_Additive
    translationSpace=BCS_BoneSpace
    alphaBoolBlend {blendOption=Linear}
    boneToModify {boneName=dyn_scarf_bk_l}
    rotation {pitch=-1}
    translation {z=1}
   }
   n21 = "Transform (Modify) Bone - Bone: dyn_scarf_bk_r"!(ComponentPose=@n20, bAlphaBoolEnabled:bool=True, Translation="0,0,1", Rotation="-1,0,0", Scale="1,1,1") -> (Pose) {
    alphaInputType=Bool
    rotationMode=BMM_Additive
    rotationSpace=BCS_BoneSpace
    scaleMode=BMM_Additive
    scaleSpace=BCS_BoneSpace
    translationMode=BMM_Additive
    translationSpace=BCS_BoneSpace
    alphaBoolBlend {blendOption=Linear}
    boneToModify {boneName=dyn_scarf_bk_r}
    rotation {pitch=-1}
    translation {z=1}
   }
   n22 = Component To Local(ComponentPose=@n21)
   n23 = Use cached pose 'PreCach'()
   n24 = Get Cherry() -> (Cherry:bool)
   n25 = "Constraint - dyn_scarf_bk_r (P)"(ComponentPose=Local To Component(LocalPose=Use cached pose 'PreCach'()), Alpha=1, ConstraintWeights_0=1) {
    alphaBoolBlend {blendOption=Linear}
    boneToModify {boneName=dyn_scarf_bk_r}
    constraintSetup [{targetBone {boneName=clavicle_r}}]
    constraintWeights [1]
   }
   n26 = "Constraint - dyn_scarf_bk_l (P)"(ComponentPose=@n25, Alpha=1, ConstraintWeights_0=1) {
    alphaBoolBlend {blendOption=Linear}
    boneToModify {boneName=dyn_scarf_bk_l}
    constraintSetup [{targetBone {boneName=clavicle_l}}]
    constraintWeights [1]
   }
   n27 = "Constraint - dyn_scarf_bk (P)"(ComponentPose=@n26, Alpha=1, ConstraintWeights_0=1) {
    alphaBoolBlend {blendOption=Linear}
    boneToModify {boneName=dyn_scarf_bk}
    constraintSetup [{targetBone {boneName=spine_05}}]
    constraintWeights [1]
   }
   n28 = "Constraint - dyn_scarf_bk_r (P)"(ComponentPose=Local To Component(LocalPose=Use cached pose 'PreCach'()), Alpha=1, ConstraintWeights_0=1) {
    alphaBoolBlend {blendOption=Linear}
    boneToModify {boneName=dyn_scarf_bk_r}
    constraintSetup [{targetBone {boneName=clavicle_r}}]
    constraintWeights [1]
   }
   n29 = "Constraint - dyn_scarf_bk_l (P)"(ComponentPose=@n28, Alpha=1, ConstraintWeights_0=1) {
    alphaBoolBlend {blendOption=Linear}
    boneToModify {boneName=dyn_scarf_bk_l}
    constraintSetup [{targetBone {boneName=clavicle_l}}]
    constraintWeights [1]
   }
   n30 = "Constraint - dyn_scarf_bk (P)"(ComponentPose=@n29, Alpha=1, ConstraintWeights_0=1) {
    alphaBoolBlend {blendOption=Linear}
    boneToModify {boneName=dyn_scarf_bk}
    constraintSetup [{targetBone {boneName=spine_05}}]
    constraintWeights [1]
   }
   n31 = Blend Poses by bool(
    BlendPose_0=Component To Local(ComponentPose=@n30)
    BlendPose_1=Blend Poses by bool(
     BlendPose_0=Component To Local(ComponentPose=@n27)
     BlendPose_1=Blend Poses by bool(BlendPose_0=@n22, BlendPose_1=@n23, bActiveValue=@n24) {blendType=Linear}
     bActiveValue=Get Fig() -> (Fig:bool)
    ) {blendType=Linear}
    bActiveValue=Get Pear() -> (Pear:bool)
   ) {blendType=Linear}
   n32 = "Transform (Modify) Bone - Bone: dyn_scarf_bk_l_MSTR"(
    ComponentPose=Local To Component(LocalPose=Use cached pose 'PreCach'())
    bAlphaBoolEnabled=True
    Translation="0,0,3"
    Rotation="-55,0,0"
    Scale="1,1,1"
   ) {
    alphaInputType=Bool
    rotationMode=BMM_Additive
    rotationSpace=BCS_BoneSpace
    scaleMode=BMM_Additive
    scaleSpace=BCS_BoneSpace
    translationMode=BMM_Additive
    translationSpace=BCS_BoneSpace
    alphaBoolBlend {blendInTime=0.1, blendOption=Linear, blendOutTime=0.1}
    boneToModify {boneName=dyn_scarf_bk_l_MSTR}
    rotation {pitch=-55}
    translation {z=3}
   }
   n33 = "Transform (Modify) Bone - Bone: dyn_scarf_bk_MSTR"(ComponentPose=@n32, bAlphaBoolEnabled=True, Translation="0,0,3", Rotation="-55,0,0", Scale="1,1,1") {
    alphaInputType=Bool
    rotationMode=BMM_Additive
    rotationSpace=BCS_BoneSpace
    scaleMode=BMM_Additive
    scaleSpace=BCS_BoneSpace
    translationMode=BMM_Additive
    translationSpace=BCS_BoneSpace
    alphaBoolBlend {blendInTime=0.1, blendOption=Linear, blendOutTime=0.1}
    boneToModify {boneName=dyn_scarf_bk_MSTR}
    rotation {pitch=-55}
    translation {z=3}
   }
   n34 = "Transform (Modify) Bone - Bone: dyn_scarf_bk_r_MSTR"(ComponentPose=@n33, bAlphaBoolEnabled=True, Translation="0,0,3", Rotation="-55,0,0", Scale="1,1,1") {
    alphaInputType=Bool
    rotationMode=BMM_Additive
    rotationSpace=BCS_BoneSpace
    scaleMode=BMM_Additive
    scaleSpace=BCS_BoneSpace
    translationMode=BMM_Additive
    translationSpace=BCS_BoneSpace
    alphaBoolBlend {blendInTime=0.1, blendOption=Linear, blendOutTime=0.1}
    boneToModify {boneName=dyn_scarf_bk_r_MSTR}
    rotation {pitch=-55}
    translation {z=3}
   }
   n35 = Component To Local(ComponentPose=@n34)
   n36 = Get Guava() -> (Guava:bool)
   n37 = Get Wearing Backbling() -> (Wearing Backbling:bool)
   n38 = "Constraint - dyn_scarf_bk_l (P)"!(
    ComponentPose=Local To Component(LocalPose=Blend Poses by bool(BlendPose_0=@n35, BlendPose_1=@n31, bActiveValue=@n36) {blendType=Linear})
    bAlphaBoolEnabled:bool=@n37
    ConstraintWeights_0=1
   ) -> (Pose) {
    alphaInputType=Bool
    alphaBoolBlend {blendOption=Linear}
    boneToModify {boneName=dyn_scarf_bk_l}
    constraintSetup [{targetBone {boneName=spine_05}}]
    constraintWeights [1]
   }
   n39 = "Constraint - dyn_scarf_bk (P)"!(ComponentPose=@n38, bAlphaBoolEnabled:bool=@n37, ConstraintWeights_0=1) -> (Pose) {
    alphaInputType=Bool
    alphaBoolBlend {blendOption=Linear}
    boneToModify {boneName=dyn_scarf_bk}
    constraintSetup [{targetBone {boneName=spine_05}}]
    constraintWeights [1]
   }
   n40 = "Constraint - dyn_scarf_bk_r (P)"!(ComponentPose=@n39, bAlphaBoolEnabled:bool=@n37, ConstraintWeights_0=1) -> (Pose) {
    alphaInputType=Bool
    alphaBoolBlend {blendOption=Linear}
    boneToModify {boneName=dyn_scarf_bk_r}
    constraintSetup [{targetBone {boneName=spine_05}}]
    constraintWeights [1]
   }
   n41 = Get Wearing Backbling()
   n42 = "Transform (Modify) Bone - Bone: dyn_scarf_bk_r"(ComponentPose=@n40, bAlphaBoolEnabled=@n41, Translation="0,0,2", Rotation="0,0,0", Scale="0.8,0.8,0.8") {
    alphaInputType=Bool
    rotationMode=BMM_Additive
    rotationSpace=BCS_BoneSpace
    scaleMode=BMM_Additive
    scaleSpace=BCS_BoneSpace
    translationMode=BMM_Additive
    translationSpace=BCS_BoneSpace
    alphaBoolBlend {blendOption=Linear}
    boneToModify {boneName=dyn_scarf_bk_r}
    scale=<0.8 0.8 0.8>
    translation {z=2}
   }
   n43 = "Transform (Modify) Bone - Bone: dyn_scarf_bk_l"(ComponentPose=@n42, bAlphaBoolEnabled=@n41, Translation="0,0,1", Rotation="0,0,0", Scale="0.8,0.8,0.8") {
    alphaInputType=Bool
    rotationMode=BMM_Additive
    rotationSpace=BCS_BoneSpace
    scaleMode=BMM_Additive
    scaleSpace=BCS_BoneSpace
    translationMode=BMM_Additive
    translationSpace=BCS_BoneSpace
    alphaBoolBlend {blendOption=Linear}
    boneToModify {boneName=dyn_scarf_bk_l}
    scale=<0.8 0.8 0.8>
    translation {z=1}
   }
   n44 = "Transform (Modify) Bone - Bone: dyn_scarf_bk"(ComponentPose=@n43, bAlphaBoolEnabled=@n41, Translation="0,0,0", Rotation="0,0,0", Scale="0.8,0.8,0.8") {
    alphaInputType=Bool
    rotationMode=BMM_Additive
    rotationSpace=BCS_BoneSpace
    scaleMode=BMM_Additive
    scaleSpace=BCS_BoneSpace
    translationMode=BMM_Additive
    translationSpace=BCS_BoneSpace
    alphaBoolBlend {blendOption=Linear}
    boneToModify {boneName=dyn_scarf_bk}
    scale=<0.8 0.8 0.8>
   }
   n45 = "Transform (Modify) Bone - Bone: dyn_scarf_bk_r"(ComponentPose=@n44, bAlphaBoolEnabled=True, Translation="0,0,2", Rotation="0,0,0", Scale="0.8,0.8,0.8") {
    alphaInputType=Bool
    rotationMode=BMM_Additive
    rotationSpace=BCS_BoneSpace
    scaleMode=BMM_Additive
    scaleSpace=BCS_BoneSpace
    translationMode=BMM_Additive
    translationSpace=BCS_BoneSpace
    alphaBoolBlend {blendOption=Linear}
    boneToModify {boneName=dyn_scarf_bk_r}
    scale=<0.8 0.8 0.8>
    translation {z=2}
   }
   n46 = "Transform (Modify) Bone - Bone: dyn_scarf_bk_l"(ComponentPose=@n45, bAlphaBoolEnabled=True, Translation="0,0,2", Rotation="0,0,0", Scale="0.8,0.8,0.8") {
    alphaInputType=Bool
    rotationMode=BMM_Additive
    rotationSpace=BCS_BoneSpace
    scaleMode=BMM_Additive
    scaleSpace=BCS_BoneSpace
    translationMode=BMM_Additive
    translationSpace=BCS_BoneSpace
    alphaBoolBlend {blendOption=Linear}
    boneToModify {boneName=dyn_scarf_bk_l}
    scale=<0.8 0.8 0.8>
    translation {z=2}
   }
   n47 = "Transform (Modify) Bone - Bone: dyn_scarf_bk"(ComponentPose=@n46, bAlphaBoolEnabled=True, Translation="0,0,2", Rotation="0,0,0", Scale="0.8,0.8,0.8") {
    alphaInputType=Bool
    rotationMode=BMM_Additive
    rotationSpace=BCS_BoneSpace
    scaleMode=BMM_Additive
    scaleSpace=BCS_BoneSpace
    translationMode=BMM_Additive
    translationSpace=BCS_BoneSpace
    alphaBoolBlend {blendOption=Linear}
    boneToModify {boneName=dyn_scarf_bk}
    scale=<0.8 0.8 0.8>
    translation {z=2}
   }
   Output Pose(Result=Component To Local(ComponentPose=@n47)) {layerGroup=None}
   n48 = Control Rig(Source=Copy Pose From Mesh() {bCopyCurves=true, bUseAttachedParent=true}) {
    alphaBoolBlend {blendOption=Linear}
    controlRigAssetReference {blueprintRigClass=&/Pencil_Sailor_MedB_Ubdy_DeformRig.Pencil_Sailor_MedB_Ubdy_DeformRig_C}
   }
   n49 = Control Rig!(
    Source=@n48
    Alpha
    crouching:double=Get crouch() -> (crouch:double)
    crouch strafe r:double=Get strafe l()
    crouch strafe l:double=Get strafe r()
    moving 2d:double=Get Moving 2D() -> (Moving 2D:double)
    emote:double=Get emote() -> (emote:double)
   ) -> (Pose) {
    alphaBoolBlend {blendOption=Linear}
    controlRigAssetReference {blueprintRigClass=/BRCosmetics/Characters/Player/Female/Medium/Bodies/F_MED_Pencil_B/SkirtLong/Meshes/Pencil_SkirtLong_MedB_Lbdy_CtrlRig.Pencil_SkirtLong_MedB_Lbdy_CtrlRig_C}
   }
   n50 = RigidBody(ComponentPose=Local To Component(LocalPose=@n49), OverrideWorldGravity="0,0,-980") {
    bClampLinearTranslationLimitToRefPose=true
    bOverrideWorldGravity=true
    overridePhysicsAsset=&/Pencil_Sailor_MedB_Ubdy_Physics
    simulationSpace=BaseBoneSpace
    alphaBoolBlend {blendOption=Linear}
    baseBoneRef {boneName=spine_05}
    componentAppliedLinearAccClamp=<3000 3000 3000>
    componentLinearAccScale=<0.48 0.48 0.6>
    componentLinearVelScale=<0.24 0.24 0.3>
    overrideWorldGravity {z=-980}
    simSpaceSettings {maxAngularVelocity=3, maxLinearAcceleration=700, maxLinearVelocity=500, velocityScaleZ=0, worldAlpha=0.6, externalLinearDragV=<0.2 0.2>}
   }
   n51 = "Constraint - dyn_scarf_fr_MSTR (PP)"!(ComponentPose=@n50, Alpha=1, ConstraintWeights_0=0.7647, ConstraintWeights_1:float=0.2353) -> (Pose) {
    alphaBoolBlend {blendOption=Linear}
    boneToModify {boneName=dyn_scarf_fr_MSTR}
    constraintSetup [{targetBone {boneName=spine_04}}, {targetBone {boneName=spine_05}}]
    constraintWeights [1,1]
   }
   n52 = "Transform (Modify) Bone - Bone: dyn_scarf_fr_MSTR"(ComponentPose=@n51, Alpha=1, Translation="0,0,0", Rotation="20,0,0", Scale="0.8,1,1") {
    rotationMode=BMM_Additive
    rotationSpace=BCS_BoneSpace
    scaleMode=BMM_Additive
    scaleSpace=BCS_BoneSpace
    translationMode=BMM_Additive
    translationSpace=BCS_BoneSpace
    alphaBoolBlend {blendOption=Linear}
    boneToModify {boneName=dyn_scarf_fr_MSTR}
    rotation {pitch=20}
    scale {x=0.8}
   }
   n53 = "Transform (Modify) Bone - Bone: dyn_scarf_bk_l_MSTR"(ComponentPose=@n52, bAlphaBoolEnabled=True, Translation="0,0,0", Rotation="30,0,0", Scale="1,1,1") {
    alphaInputType=Bool
    rotationMode=BMM_Additive
    rotationSpace=BCS_BoneSpace
    scaleMode=BMM_Additive
    scaleSpace=BCS_BoneSpace
    translationMode=BMM_Additive
    translationSpace=BCS_BoneSpace
    alphaBoolBlend {blendInTime=0.1, blendOption=Linear, blendOutTime=0.1}
    boneToModify {boneName=dyn_scarf_bk_l_MSTR}
    rotation {pitch=30}
   }
   n54 = "Transform (Modify) Bone - Bone: dyn_scarf_bk_MSTR"(ComponentPose=@n53, bAlphaBoolEnabled=True, Translation="0,0,0", Rotation="30,0,0", Scale="1,1,1") {
    alphaInputType=Bool
    rotationMode=BMM_Additive
    rotationSpace=BCS_BoneSpace
    scaleMode=BMM_Additive
    scaleSpace=BCS_BoneSpace
    translationMode=BMM_Additive
    translationSpace=BCS_BoneSpace
    alphaBoolBlend {blendInTime=0.1, blendOption=Linear, blendOutTime=0.1}
    boneToModify {boneName=dyn_scarf_bk_MSTR}
    rotation {pitch=30}
   }
   n55 = "Transform (Modify) Bone - Bone: dyn_scarf_bk_r_MSTR"(ComponentPose=@n54, bAlphaBoolEnabled=True, Translation="0,0,0", Rotation="30,0,0", Scale="1,1,1") {
    alphaInputType=Bool
    rotationMode=BMM_Additive
    rotationSpace=BCS_BoneSpace
    scaleMode=BMM_Additive
    scaleSpace=BCS_BoneSpace
    translationMode=BMM_Additive
    translationSpace=BCS_BoneSpace
    alphaBoolBlend {blendInTime=0.1, blendOption=Linear, blendOutTime=0.1}
    boneToModify {boneName=dyn_scarf_bk_r_MSTR}
    rotation {pitch=30}
   }
   Save cached pose 'PreCach'(Pose=Component To Local(ComponentPose=@n55)) {cacheName=PreCach}
  ]}
 EventGraph {nodes [
   Event BlueprintUpdateAnimation()
   > n1 = Sequence!(execute) -> (then_0, then_1, then_2:Exec, then_3:Exec, then_4:Exec, then_5:Exec)
   > n2 = Gravity Override State Function DEPRECATED(CharacterPartAnimInstance=Self-Reference(), GameTimeInSeconds=GetGameTimeInSeconds(), GravityOverrideParams=Get Gravity Override Params())
   > Set GravityOverride(GravityOverride=@n2)
   n3 = GetOwnerSkeletalMeshComponent()
   n4 = BreakRotator(InRot=BreakTransform(InTransform=GetDeltaTransformFromRefPose(self=@n3, BoneName=upperarm_l, BaseName=spine_05)).Rotation)
   Set arm_rot_up_L(execute=@n1.then_1, arm_rot_up_L:double=MapRangeClamped(Value="Absolute (Float)"(A=@n4.Yaw), InRangeA=50, InRangeB=130)) -> (Output_Get:double)
   n5 = PlayerMovement(execute=@n1.then_2, AnimInstance=Self-Reference(), XDotProductCoefficient=1, YDotProductCoefficient=1, ZDotProductCoefficient=1)
   > Set strafe r(strafe r:double=MapRangeClamped(Value=@n5, InRangeA=0, InRangeB=1)) -> (Output_Get:double)
   > Set strafe l(strafe l:double=MapRangeClamped(Value=@n5, InRangeA=0, InRangeB=-1)) -> (Output_Get:double)
   > n6 = Branch(Condition=Get bIsMoving2D())
   > Set Moving 2D(Moving 2D:double=1) -> (Output_Get:double)
   Set Moving 2D(execute=@n6.else, Moving 2D=0)
   Set Fig(execute=@n1.then_3, Fig:bool=CharacterItemDefinitionHasAnyMetaTag(Tags=Make Array("[0]"=(TagName="Cosmetics.MetaTags.Pencil.Head.Fig")))) -> (Output_Get:bool)
   > Set Cherry(Cherry:bool=CharacterItemDefinitionHasAnyMetaTag(Tags=Make Array("[0]"=(TagName="Cosmetics.MetaTags.Pencil.Head.Cherry")))) -> (Output_Get:bool)
   > Set Pear(Pear:bool=CharacterItemDefinitionHasAnyMetaTag(Tags=Make Array("[0]"=(TagName="Cosmetics.MetaTags.Pencil.Head.Pear")))) -> (Output_Get:bool)
   > Set Guava(Guava:bool=CharacterItemDefinitionHasAnyMetaTag(Tags=Make Array("[0]"=(TagName="Cosmetics.MetaTags.Pencil.Head.Guava")))) -> (Output_Get:bool)
   n7 = CheckForBackbling()
   Set Wearing Backbling(execute=@n1.then_4, Wearing Backbling:bool=OR Boolean!(A=@n7.bIsBackblingACape, B=@n7.bIsBackblingLong, C:bool=@n7) -> (ReturnValue)) -> (Output_Get:bool)
   n8 = GetOwnerSkeletalMeshComponent()
   n9 = BreakRotator(InRot=BreakTransform(InTransform=GetDeltaTransformFromRefPose(self=@n8, BoneName=upperarm_r, BaseName=spine_05)).Rotation)
   Set Arm Rot Up R(execute=@n1.then_5, Arm Rot Up R:double=MapRangeClamped(Value="Absolute (Float)"(A=@n9.Yaw), InRangeA=50, InRangeB=130)) -> (Output_Get:double)
   Event OnBeginSkydiving()
   > Set skydive(skydive:double=0.4) -> (Output_Get:double)
   Event OnBeginPlayingEmote() {is=title K2Node_Event}
   > Set emote(emote:double=1) -> (Output_Get:double)
   Event OnEndPlayingEmote() {is=title K2Node_Event}
   > Set emote(emote=0)
   Event OnEndSkydiving()
   > Set skydive(skydive=0)
   n10 = Event OnEndFalling()
   Set jumping(execute=[@n10.then, @n12.then], jumping:double=0) -> (Output_Get:double)
   n11 = Event OnBeginJumping()
   Set jumping(execute=[@n11.then, @n13.then], jumping=1)
   n12 = Event OnEndJumping()
   Event OnEndCrouching()
   > Set crouch(crouch:double=0) -> (Output_Get:double)
   n13 = Event OnBeginFalling()
   Event OnBeginCrouching()
   > Set crouch(crouch=1)
  ]}
]
variables [
 Arm Rot Up R {type=number}
 arm_rot_up_L {type=number}
 Cherry {type=boolean}
 crouch {type=number}
 emote {type=number}
 Fig {type=boolean}
 Gravity Override Params {
  type=GravityOverrideParamsStruct
  default {skydive_FloatingDown {gravityOverride=<0 -700 -980>}, skydive_GoingUp {gravityOverride=<0 -300 -980>}}
 }
 GravityOverride {type=Vector}
 Guava {type=boolean}
 jumping {type=number}
 Moving 2D {type=number}
 Pear {type=boolean}
 skydive {type=number}
 strafe l {type=number}
 strafe r {type=number}
 Wearing Backbling {type=boolean}
]
